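# UI material for the Crew series cosmetic background, drawing an animated field of rounded boxes over a blue gradient.
# Two panning, rotated particle-grid layers, weighted by per-cell random values and tinted, are summed to blend a rotated dark-to-blue gradient toward periwinkle.
# Vertex alpha modulates the result, and a rounded-corner mask with per-corner radii, optionally in pixels, is scaled by Additive Percentage for alpha-composite blending.
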Material /BRCosmetics/Athena/UI/Series/Art/M_UI_Crew_Series
graph {
 properties {
  bAutomaticallySetUsageInEditor=false
  bUsedWithStaticMesh=false
  blendMode=BLEND_AlphaComposite
  materialDomain=MD_UI
  usedShadingModels=Default Lit
 }
 nodes [
  n1 = LinearInterpolate(
   A=Constant3Vector() {constant=<rgba 0.1306 0.1893 0.6319 1>}
   B=Constant3Vector() {constant=<rgba 0.0145 0.0119 0.0434 1>}
   Alpha=MaterialFunctionCall(UVs=NamedRerouteUsage(), RotationAngle=Constant() {r=0.75}) {materialFunction=&MF_UI_LinearGradient}.Result
  )
  n2 = Constant3Vector() {constant=<rgba 0.2862 0.3432 0.9473 1>}
  n3 = Constant()
  n4 = MaterialFunctionCall!(
   UV=Panner(Coordinate=NamedRerouteUsage()) {speedX=0.032, speedY=0.01}
   Global Offset=Constant2Vector() {g=0.3, r=0.3}
   Global Rotation=Constant() {r=0.35}
   Tiling=Constant() {r=3}
   Cell Offset=@n3
   Cell Scale=LinearInterpolate(Alpha=NamedRerouteUsage()) {constA=0.1, constB=0.85}
   Cell Rotation=@n3
   Random Seed=Constant() {r=0.1}
  ) {materialFunction=/FortUILibrary/MaterialFunctions/Basics/MF_UI_ParticleGrid}
  n5 = Multiply(
   A=MaterialFunctionCall(UVs=@n4.Cells UV, Size=Constant() {r=0.8}, CornerRoundness=Constant() {r=0.25}, GlowMax=Constant() {r=0.5}, KeepAspectRatio=StaticBool()) {materialFunction=&MF_UI_Box}.Fill
   B=NamedRerouteDeclaration(None=@n4.Random Cell Values) {name=Random Cell Values_1, nodeColor=<rgba 1 0 0.8672 1>}
  )
  n6 = Constant3Vector() {constant=<rgba 0.162 0.1683 0.4125 1>}
  n7 = Constant()
  n8 = MaterialFunctionCall(
   UV=Panner(Coordinate=NamedRerouteUsage()) {speedX=0.025, speedY=0.01}
   Global Rotation=Constant() {r=0.35}
   Tiling=Constant() {r=4}
   Cell Offset=@n7
   Cell Scale=LinearInterpolate(Alpha=NamedRerouteUsage()) {constA=0.5, constB=0.75}
   Cell Rotation=@n7
   Random Seed=Constant() {r=0.2}
  ) {materialFunction=&MF_UI_ParticleGrid}
  n9 = Multiply(
   A=MaterialFunctionCall(UVs=@n8.Cells UV, Size=Constant() {r=1}, CornerRoundness=Constant() {r=0.25}, KeepAspectRatio=StaticBool()) {materialFunction=&MF_UI_Box}.Fill
   B=NamedRerouteDeclaration(None=@n8.Random Cell Values) {name=Random Cell Values, nodeColor=<rgba 1 0 0.8672 1>}
  )
  n10 = Saturate(None=Add(A=Multiply(A=@n5, B=@n6), B=Multiply(A=@n9, B=@n2)))
  Multiply(A=Multiply(A=LinearInterpolate(A=@n1, B=@n2, Alpha=@n10), B=VertexColor().A), B=NamedRerouteUsage())
  n11 = MaterialFunctionCall(
   R=ScalarParameter() {group=Layout, parameterName=TopLeftCornerRadius, sortPriority=4}
   G=ScalarParameter() {group=Layout, parameterName=BottomLeftCornerRadius, sortPriority=7}
   B=ScalarParameter() {group=Layout, parameterName=TopRightCornerRadius, sortPriority=5}
   A=ScalarParameter() {group=Layout, parameterName=BottomRightCornerRadius, sortPriority=6}
  ) {materialFunction=&AppendMany}
  n12 = ScalarParameter() {
   defaultValue=1
   desc="If enabled reads corner radius values in pixels.\r\nbool"
   group=Layout
   parameterName=UseValueInPixels
   sliderMax=1
   sortPriority=1
  }
  NamedRerouteDeclaration(None=Multiply(
    A=MaterialFunctionCall(Corner Radius=@n11.RGBA, Radius in Pixels=@n12) {materialFunction=&MF_UI_CoreUI_RoundedMask}.Shape Mask
    B=ScalarParameter() {defaultValue=1, group=Color, parameterName=Additive Percentage, sliderMax=1, sortPriority=3}
   )) {name=ShapeMask, nodeColor=<rgba 0 0 0 1>}
  NamedRerouteDeclaration(None=MaterialFunctionCall() {materialFunction=&MF_UI_UVScaling}.ScaleY) {name=UVScaleY, nodeColor=<rgba 0.1719 0 1 1>}
 ]
}
