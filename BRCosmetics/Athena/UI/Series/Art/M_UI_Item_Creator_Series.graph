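# UI material that draws the animated dark-blue to cyan background for Creator Series items.
# Two panning screen-aligned noise samples distort a radial gradient into clouds, layered over aspect-corrected background and overlay gradients.
# Two sets of sine-driven angular beams sway, stretch and fade radially, picking up the cloud color and blending toward bright cyan.
# A rounded-corner shape mask, scaled by Additive Percentage, and vertex alpha multiply the color for alpha-composite blending.
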Material /BRCosmetics/Athena/UI/Series/Art/M_UI_Item_Creator_Series
graph {
 properties {
  bUsedWithStaticMesh=false
  blendMode=BLEND_AlphaComposite
  materialDomain=MD_UI
  pixelDepthOffsetMode=PDOM_Legacy
  usedShadingModels=Default Lit
 }
 nodes [
  n1 = MaterialFunctionCall(Float2=MaterialFunctionCall() {materialFunction=&ScreenResolution}.Visible Resolution) {materialFunction=&BreakOutFloat2Components}
  n2 = Max(A=@n1.G)
  n3 = Multiply(
   A=AppendVector(A=Divide(A=@n1.R, B=@n2), B=Constant() {r=1})
   B=MaterialFunctionCall() {materialFunction=&ScreenAlignedUVs}."X 100%, Y 100%"
  )
  n4 = Add(
   A=Multiply(A=@n3, B=ScalarParameter() {defaultValue=4, group=CC-Series-Cloud, parameterName=Cloud-Beam-Global-U-Tiles})
   B=Multiply(A=TextureCoordinate(), B=ScalarParameter() {defaultValue=0.5, group=CC-Series-Cloud, parameterName=Cloud-Beam-Localized UV})
  )
  n5 = Multiply(A=Time(), B=ScalarParameter() {defaultValue=0.25, group=CC-Series-Cloud, parameterName=Time Factor})
  n6 = Add(A=Multiply(A=@n3, B=ScalarParameter() {defaultValue=4, group=CC-Series-Cloud, parameterName=Cloud-Beam-Global-U-Tiles}))
  n7 = Add(
   A=TextureSample(UVs=Panner(Coordinate=@n4, Time=@n5) {speedX=0.35, speedY=0.125}) {samplerType=SAMPLERTYPE_Grayscale, texture=&LowResBlurredNoise}.R
   B=TextureSample(UVs=Panner(Coordinate=@n6, Time=@n5) {speedX=0.25, speedY=0.25}) {samplerType=SAMPLERTYPE_Grayscale, texture=&LowResBlurredNoise}.R
  )
  n8 = ScalarParameter() {defaultValue=0.35, group=CC-Series-Cloud, parameterName=Cloud_Multiply}
  n9 = MaterialFunctionCall(
   UVs=Add(A=TextureCoordinate(), B=Multiply(A=@n7, B=@n8))
   CenterPosition=AppendVector(
    A=ScalarParameter() {defaultValue=0.75, group=CC-Series-Cloud, parameterName=Cloud-X-Offset}
    B=ScalarParameter() {defaultValue=0.25, group=CC-Series-Cloud, parameterName=Cloud-Y-Offset}
   )
   Radius=ScalarParameter() {defaultValue=0.5, group=CC-Series-Cloud, parameterName=Radius-Cloud, sortPriority=1}
   Density=ScalarParameter() {defaultValue=1.5, group=CC-Series-Cloud, parameterName=Density-Cloud, sortPriority=1}
  ) {materialFunction=&RadialGradientExponential}
  n10 = Constant3Vector() {constant=<rgba 0.0006 0.0262 0.0578 1>}
  n11 = Constant3Vector() {constant=<rgba 0.0037 0.3079 0.3177 1>}
  n12 = Multiply(A=@n9.RadialGradientExponential, B=LinearInterpolate(A=@n10, B=@n11, Alpha=@n9.RadialGradientExponential))
  n13 = ScalarParameter() {defaultValue=0.75, group=CC-Series-Cloud, parameterName=Cloud-Intensity}
  n14 = MaterialFunctionCall() {materialFunction=&GetUserInterfaceUV}
  n15 = MaterialFunctionCall(Float2=@n14.Pixel Size) {materialFunction=&BreakOutFloat2Components}
  n16 = AppendVector(A=Constant() {r=1}, B=Divide(A=@n15.G, B=Max(A=@n15.R)))
  n17 = Add(A=Multiply(A=OneMinus(None=@n16)) {constB=0.5}, B=Multiply(A=@n16, B=@n14.Normalized UV))
  n18 = MaterialFunctionCall(
   UVs=@n17
   CenterPosition=AppendVector(
    A=ScalarParameter() {defaultValue=0.5, group=CC-Series, parameterName=X-Offset}
    B=ScalarParameter() {defaultValue=0.9, group=CC-Series, parameterName=Y-Offset}
   )
   Radius=ScalarParameter() {defaultValue=1.35, group=CC-Series, parameterName=Background Radius}
   Density=ScalarParameter() {defaultValue=1.25, group=CC-Series, parameterName=Background Density}
  ) {materialFunction=&RadialGradientExponential}
  n19 = Constant3Vector() {constant=<rgba 0.0802 0.9823 0.9823 1>}
  n20 = ScalarParameter() {defaultValue=0.5, group=CC-Series, parameterName=Overlay X-Offset}
  n21 = ScalarParameter() {defaultValue=0.925, group=CC-Series, parameterName=Overlay Y-Offset}
  n22 = LinearInterpolate(
   A=Add(A=Multiply(A=@n12, B=@n13), B=LinearInterpolate(A=@n10, B=@n11, Alpha=@n18.RadialGradientExponential))
   B=@n19
   Alpha=MaterialFunctionCall(
    UVs=@n17
    CenterPosition=AppendVector(A=@n20, B=@n21)
    Radius=ScalarParameter() {defaultValue=0.95, group=CC-Series, parameterName=Overlay Radius}
    Density=ScalarParameter() {defaultValue=2, group=CC-Series, parameterName=Overlay Density}
   ) {materialFunction=&RadialGradientExponential}.RadialGradientExponential
  )
  n23 = Multiply(A=@n12, B=ScalarParameter() {defaultValue=20, group=CC-Series-Beams, parameterName=Beam-Cloud-Intensity})
  n24 = ScalarParameter() {defaultValue=-0.5, group=CC-Series-Beams, parameterName=1st-Beam X-Offset}
  n25 = ScalarParameter() {defaultValue=-1.25, group=CC-Series-Beams, parameterName=1st-Beam Y-Offset}
  n26 = MaterialFunctionCall(
   CenterPosition=Multiply(A=AppendVector(A=@n24, B=@n25)) {constB=-1}
   Radius=ScalarParameter() {defaultValue=1.5, group=CC-Series-Beams, parameterName=1st-Beam-Radius}
   Density=ScalarParameter() {defaultValue=0.75, group=CC-Series-Beams, parameterName=1st-Beam-Density}
  ) {materialFunction=&RadialGradientExponential}
  n27 = AppendVector(
   A=ScalarParameter() {defaultValue=-0.5, group=CC-Series-Beams, parameterName=1st-Beam X-Offset}
   B=ScalarParameter() {defaultValue=-1.25, group=CC-Series-Beams, parameterName=1st-Beam Y-Offset}
  )
  n28 = Sine(None=Multiply(A=Time(), B=ScalarParameter() {defaultValue=0.25, group=CC-Series-Beams, parameterName=Beam-Scale-TimeFactor})) {period=2}
  n29 = AppendVector(A=LinearInterpolate(Alpha=Add(A=@n28) {constB=0.5}) {constA=1, constB=1.5}, B=Constant() {r=1})
  n30 = Sine(None=Multiply(A=Time(), B=ScalarParameter() {defaultValue=0.1, group=CC-Series-Beams, parameterName=1st-Beam-TimeFactor})) {period=2}
  n31 = Rotator(Coordinate=Multiply(A=Add(A=@n27, B=@n17), B=@n29), Time=LinearInterpolate(Alpha=@n30) {constA=-1}) {centerX=0, centerY=0}
  n32 = Sine(None=Multiply(
    A=MaterialFunctionCall(Vector or UVs=@n31) {materialFunction=&VectorToRadialValue}.Vector Converted to Angle
    B=ScalarParameter() {defaultValue=8, group=CC-Series-Beams, parameterName=1st-Beam-Count}
   ) {constB=2})
  n33 = LinearInterpolate(
   A=ScalarParameter() {defaultValue=-0.8, group=CC-Series-Beams, parameterName=1st-Beam-Thickness-Min}
   B=ScalarParameter() {defaultValue=-0.85, group=CC-Series-Beams, parameterName=1st-Beam-Thickness-Max}
   Alpha=@n30
  )
  n34 = Saturate(None=Multiply(A=Add(A=@n32, B=@n33) {constB=-0.25}, B=ScalarParameter() {defaultValue=4, group=CC-Series-Beams, parameterName=1st-Beam-Edge}))
  n35 = LinearInterpolate(B=Add(A=Multiply(A=@n32, B=@n26.RadialGradientExponential), B=Multiply(A=@n34, B=@n26.RadialGradientExponential)), Alpha=@n34)
  n36 = ScalarParameter() {defaultValue=1.5, group=CC-Series-Beams, parameterName=1st-Beam-Multiply}
  n37 = AppendVector(
   A=ScalarParameter() {defaultValue=-0.5, group=CC-Series-Beams, parameterName=2nd-Beam X-Offset}
   B=ScalarParameter() {defaultValue=-0.95, group=CC-Series-Beams, parameterName=2nd-Beam Y-Offset}
  )
  n38 = Sine(None=Multiply(A=Time(), B=ScalarParameter() {defaultValue=0.125, group=CC-Series-Beams, parameterName=2nd-Beam-TimeFactor})) {period=2}
  n39 = Rotator(Coordinate=Multiply(A=@n29, B=Add(A=@n17, B=@n37)), Time=LinearInterpolate(Alpha=@n38) {constA=-1}) {centerX=0, centerY=0}
  n40 = Sine(None=Multiply(
    A=MaterialFunctionCall(Vector or UVs=@n39) {materialFunction=&VectorToRadialValue}.Vector Converted to Angle
    B=ScalarParameter() {defaultValue=5, group=CC-Series-Beams, parameterName=2nd-Beam-Count}
   ) {constB=2})
  n41 = LinearInterpolate(
   A=ScalarParameter() {defaultValue=-0.775, group=CC-Series-Beams, parameterName=2nd-Beam-Thickness-Min}
   B=ScalarParameter() {defaultValue=-0.8, group=CC-Series-Beams, parameterName=2nd-Beam-Thickness-Max}
   Alpha=@n38
  )
  n42 = Saturate(None=Multiply(A=Add(A=@n40, B=@n41) {constB=-0.25}, B=ScalarParameter() {defaultValue=8, group=CC-Series-Beams, parameterName=2nd-Beam-Edge}))
  n43 = ScalarParameter() {defaultValue=-0.5, group=CC-Series-Beams, parameterName=2nd-Beam X-Offset}
  n44 = ScalarParameter() {defaultValue=-0.95, group=CC-Series-Beams, parameterName=2nd-Beam Y-Offset}
  n45 = MaterialFunctionCall(
   CenterPosition=Multiply(A=AppendVector(A=@n43, B=@n44)) {constB=-1}
   Radius=ScalarParameter() {defaultValue=1.5, group=CC-Series-Beams, parameterName=2nd-Beam-Radius}
   Density=ScalarParameter() {defaultValue=0.75, group=CC-Series-Beams, parameterName=2nd-Beam-Density}
  ) {materialFunction=&RadialGradientExponential}
  n46 = LinearInterpolate(B=Add(A=Multiply(A=@n42, B=@n45.RadialGradientExponential), B=Multiply(A=@n40, B=@n45.RadialGradientExponential)), Alpha=@n42)
  n47 = Saturate(None=LinearInterpolate(
    A=Multiply(A=@n35, B=@n36)
    B=ScalarParameter() {defaultValue=2, group=CC-Series-Beams, parameterName=2nd-Beam-Multiply}
    Alpha=Saturate(None=@n46)
   ))
  n48 = Add(A=LinearInterpolate(B=@n23, Alpha=@n47), B=@n47)
  Multiply(A=Multiply(A=LinearInterpolate(A=@n22, B=@n19, Alpha=@n48), B=NamedRerouteUsage()), B=VertexColor().A)
  Multiply(A=@n23)
  n49 = MaterialFunctionCall(
   R=ScalarParameter() {group=Layout, parameterName=TopLeftCornerRadius, sortPriority=4}
   G=ScalarParameter() {group=Layout, parameterName=BottomLeftCornerRadius, sortPriority=7}
   B=ScalarParameter() {group=Layout, parameterName=TopRightCornerRadius, sortPriority=5}
   A=ScalarParameter() {group=Layout, parameterName=BottomRightCornerRadius, sortPriority=6}
  ) {materialFunction=&AppendMany}
  n50 = ScalarParameter() {
   defaultValue=1
   desc="If enabled reads corner radius values in pixels.\r\nbool"
   group=Layout
   parameterName=UseValueInPixels
   sliderMax=1
   sortPriority=1
  }
  NamedRerouteDeclaration(None=Multiply(
    A=MaterialFunctionCall(Corner Radius=@n49.RGBA, Radius in Pixels=@n50) {materialFunction=&MF_UI_CoreUI_RoundedMask}.Shape Mask
    B=ScalarParameter() {defaultValue=1, group=Color, parameterName=Additive Percentage, sliderMax=1, sortPriority=3}
   )) {name=ShapeMask, nodeColor=<rgba 0 0 0 1>}
  NamedRerouteUsage()
 ]
}
