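# UI material that draws the animated magenta-and-purple backdrop for Cube series items in Battle Royale cosmetics UI.
# An aspect-corrected radial gradient, distorted by two panning ice-noise samples, sits beneath rear and front layers of scrolling glyphs masked by cube patterns.
# Pattern UVs mix screen-aligned and local coordinates.
# A Background parameter blends toward the SeriesTexture image, and a rounded-corner mask and vertex alpha shape the alpha-composite result.
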Material /BRCosmetics/Athena/UI/Series/Art/M_UI_Item_Cube_Series
graph {
 properties {
  bUsedWithStaticMesh=false
  blendMode=BLEND_AlphaComposite
  materialDomain=MD_UI
  pixelDepthOffsetMode=PDOM_Legacy
  usedShadingModels=Default Lit
 }
 nodes [
  n1 = Constant3Vector() {constant=<rgba 1 0.0782 0.5583 1>}
  n2 = Multiply(A=@n1, B=ScalarParameter() {defaultValue=1, group=Cube-Series, parameterName=RearGlyphIntensity})
  n3 = MaterialFunctionCall(Float2=MaterialFunctionCall() {materialFunction=&ScreenResolution}.Visible Resolution) {materialFunction=&BreakOutFloat2Components}
  n4 = Max(A=@n3.G)
  n5 = Multiply(
   A=AppendVector(A=Divide(A=@n3.R, B=@n4), B=Constant() {r=1})
   B=MaterialFunctionCall() {materialFunction=&ScreenAlignedUVs}."X 100%, Y 100%"
  )
  n6 = Multiply(A=@n5, B=ScalarParameter() {defaultValue=2, group=Cube-Series, parameterName=Global UV Foreground})
  n7 = Add(A=Multiply(A=TextureCoordinate(), B=ScalarParameter() {defaultValue=0.75, group=Cube-Series, parameterName=LocalizedUV-Add}), B=@n6)
  n8 = ScalarParameter() {defaultValue=0.5, group=Cube-Series, parameterName=RearGlyphMaskSpeedOffset}
  n9 = Time()
  n10 = Multiply(
   A=Panner(Coordinate=@n7, Time=Multiply(A=@n8, B=@n9)) {speedY=0.25}
   B=ScalarParameter() {defaultValue=0.5, group=Cube-Series, parameterName=RearGlyphMask}
  )
  n11 = ScalarParameter() {defaultValue=0.125, group=Cube-Series, parameterName=RearGlyphMaskTimeFactor}
  n12 = Multiply(
   A=Panner(Coordinate=@n7, Time=Multiply(A=@n9, B=@n11)) {speedY=0.25}
   B=ScalarParameter() {defaultValue=2, group=Cube-Series, parameterName=RearGlyphTiles}
  )
  n13 = Multiply(
   A=Power(
    Base=TextureSample(UVs=@n10) {samplerType=SAMPLERTYPE_LinearColor, texture=/Game/UI/Foundation/Textures/Pattern/T_CubePattern}.R
    Exp=ScalarParameter() {defaultValue=8, group=Cube-Series, parameterName=RearGlyphMaskPower}
   )
   B=Power(
    Base=TextureSample(UVs=@n12) {samplerType=SAMPLERTYPE_LinearColor, texture=/Game/UI/Foundation/Textures/Pattern/T_GlyphPattern}.B
    Exp=ScalarParameter() {defaultValue=1, group=Cube-Series, parameterName=RearGlyphPower}
   )
  )
  n14 = MaterialFunctionCall(Float2=TextureCoordinate() {coordinateIndex=3}) {materialFunction=&BreakOutFloat2Components}
  n15 = Divide(A=@n14.R, B=Max(A=@n14.G))
  n16 = Constant() {r=1}
  n17 = OneMinus(None=@n15)
  n18 = Add(
   A=Multiply(A=AppendVector(A=@n15, B=@n16), B=TextureCoordinate() {coordinateIndex=4})
   B=AppendVector(A=Multiply(A=@n17) {constB=0.5}, B=Constant())
  )
  n19 = Multiply(A=TextureCoordinate(), B=ScalarParameter() {defaultValue=0.25, group=Cube-Series, parameterName=Localized UV})
  n20 = Add(A=Multiply(A=@n5, B=ScalarParameter() {defaultValue=2, group=Cube-Series, parameterName=Global UV Background}), B=@n19)
  n21 = Multiply(A=Time(), B=ScalarParameter() {defaultValue=0.25, group=Cube-Series, parameterName=Time Factor})
  n22 = Panner(Coordinate=@n20, Time=@n21) {speedX=0.25, speedY=0.25}
  n23 = Panner(Coordinate=@n20, Time=@n21) {speedX=0.35, speedY=0.125}
  n24 = Multiply(
   A=Add(
    A=TextureSample(UVs=@n22) {texture=/Game/Global/Textures/Default/Noise/T_UI_Ice_Noise}.R
    B=TextureSample(UVs=@n23) {texture=&T_UI_Ice_Noise}.R
   )
   B=ScalarParameter() {defaultValue=0.125, group=Cube-Series, parameterName=Global Intensity}
  )
  n25 = LinearInterpolate(
   A=Constant3Vector() {constant=<rgba 0.008 0.0012 0.0222 1>}
   B=Constant3Vector() {constant=<rgba 0.3785 0 0.3916 1>}
   Alpha=MaterialFunctionCall(
    UVs=Add(A=@n24, B=@n18)
    CenterPosition=Constant2Vector() {g=0.5, r=0.5}
    Radius=ScalarParameter() {defaultValue=0.6, group=Cube-Series, parameterName=Backgroung-Radius-Enhanced}
    Density=ScalarParameter() {defaultValue=1.5, group=Cube-Series, parameterName=Background-Density-Enhanced}
   ) {materialFunction=&RadialGradientExponential}.RadialGradientExponential
  )
  n26 = ScalarParameter() {defaultValue=2, group=Cube-Series, parameterName=GlyphIntensity}
  n27 = Add(A=@n6, B=@n19)
  n28 = Time()
  n29 = Panner(Coordinate=@n27, Time=Multiply(A=ScalarParameter() {defaultValue=1.5, group=Cube-Series, parameterName=GlyphMaskSpeedOffset}, B=@n28)) {speedY=0.25}
  n30 = ScalarParameter() {defaultValue=0.3125, group=Cube-Series, parameterName=GlyphMask}
  n31 = Power(
   Base=TextureSample(UVs=Multiply(A=@n29, B=@n30)) {samplerType=SAMPLERTYPE_LinearColor, texture=&T_CubePattern}.R
   Exp=ScalarParameter() {defaultValue=10, group=Cube-Series, parameterName=GlyphMaskPower}
  )
  n32 = Panner(Coordinate=@n27, Time=Multiply(A=@n28, B=ScalarParameter() {defaultValue=0.25, group=Cube-Series, parameterName=GlyphMaskTimeFactor})) {speedY=0.25}
  n33 = ScalarParameter() {defaultValue=1.25, group=Cube-Series, parameterName=GlyphTiles}
  n34 = Power(
   Base=TextureSample(UVs=Multiply(A=@n32, B=@n33)) {samplerType=SAMPLERTYPE_LinearColor, texture=&T_GlyphPattern}.B
   Exp=ScalarParameter() {defaultValue=4, group=Cube-Series, parameterName=GlyphPower}
  )
  n35 = MaterialFunctionCall(Float2=TextureCoordinate() {coordinateIndex=3}) {materialFunction=&BreakOutFloat2Components}
  n36 = Divide(A=@n35.R, B=Max(A=@n35.G))
  n37 = Multiply(A=AppendVector(A=@n36, B=Constant() {r=1}), B=TextureCoordinate() {coordinateIndex=4})
  n38 = AppendVector(A=Multiply(A=OneMinus(None=@n36)) {constB=0.5}, B=Constant())
  n39 = LinearInterpolate(
   A=LinearInterpolate(A=LinearInterpolate(A=@n25, B=@n2, Alpha=@n13), B=Multiply(A=@n26, B=@n1), Alpha=Multiply(A=@n31, B=@n34))
   B=TextureSampleParameter2D(UVs=Add(A=@n37, B=@n38)) {group=Series, parameterName=SeriesTexture, texture=/Game/Athena/UI/Series/Art/DCU-Series/T-Cube-Background}.RGB
   Alpha=ScalarParameter() {parameterName=Background}
  )
  Multiply(A=Multiply(A=Multiply(A=@n39), B=NamedRerouteUsage()), B=VertexColor().A)
  n40 = MaterialFunctionCall(
   R=ScalarParameter() {group=Layout, parameterName=TopLeftCornerRadius, sortPriority=4}
   G=ScalarParameter() {group=Layout, parameterName=BottomLeftCornerRadius, sortPriority=7}
   B=ScalarParameter() {group=Layout, parameterName=TopRightCornerRadius, sortPriority=5}
   A=ScalarParameter() {group=Layout, parameterName=BottomRightCornerRadius, sortPriority=6}
  ) {materialFunction=&AppendMany}
  n41 = ScalarParameter() {
   defaultValue=1
   desc="If enabled reads corner radius values in pixels.\r\nbool"
   group=Layout
   parameterName=UseValueInPixels
   sliderMax=1
   sortPriority=1
  }
  NamedRerouteDeclaration(None=Multiply(
    A=MaterialFunctionCall(Corner Radius=@n40.RGBA, Radius in Pixels=@n41) {materialFunction=&MF_UI_CoreUI_RoundedMask}.Shape Mask
    B=ScalarParameter() {defaultValue=1, group=Color, parameterName=Additive Percentage, sliderMax=1, sortPriority=3}
   )) {name=ShapeMask, nodeColor=<rgba 0 0 0 1>}
  NamedRerouteUsage()
 ]
}
