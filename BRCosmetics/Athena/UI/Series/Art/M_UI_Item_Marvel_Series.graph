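# UI material that draws the animated Marvel series background for items, with an optional series texture blended in by a Background parameter.
# Two noise channels pan as rotated, aspect-corrected streaks, with a moving sine wave modulating their power and intensity.
# The boosted streaks add red highlights over a two-color exponential radial gradient.
# A rounded-corner mask, scaled by Additive Percentage, shapes the alpha-composite result, and vertex alpha fades the color.
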Material /BRCosmetics/Athena/UI/Series/Art/M_UI_Item_Marvel_Series
graph {
 properties {
  bUsedWithStaticMesh=false
  blendMode=BLEND_AlphaComposite
  materialDomain=MD_UI
  pixelDepthOffsetMode=PDOM_Legacy
  usedShadingModels=Default Lit
 }
 nodes [
  n1 = VectorParameter() {parameterName=RGB1, defaultValue=<rgba 0.8879 0.0252 0.0168 1>}
  n2 = AppendVector(
   A=ScalarParameter() {defaultValue=0.75, group=Marvel-Series, parameterName=Streak-R-U-Tile}
   B=ScalarParameter() {defaultValue=0.25, group=Marvel-Series, parameterName=Streak-R-V-Tile}
  )
  n3 = MaterialFunctionCall() {materialFunction=&GetUserInterfaceUV}
  n4 = MaterialFunctionCall(Float2=@n3.Pixel Size) {materialFunction=&BreakOutFloat2Components}
  n5 = Max(A=@n4.G)
  n6 = Multiply(A=AppendVector(A=Divide(A=@n4.R, B=@n5), B=Constant() {r=1}), B=@n3.Normalized UV)
  n7 = Rotator(Coordinate=@n6, Time=ScalarParameter() {defaultValue=0.6, group=Marvel-Series, parameterName=Streak Angle}) {speed=6.2832}
  n8 = Multiply(A=Time(), B=ScalarParameter() {defaultValue=1, group=Marvel-Series, parameterName=Effect-Time-Factor})
  n9 = Add(
   A=Panner(Coordinate=Multiply(A=@n7, B=@n2), Time=@n8) {speedY=1}
   B=AppendVector(
    A=ScalarParameter() {defaultValue=-0.125, group=Marvel-Series, parameterName=Streak-R-U-Offset}
    B=ScalarParameter() {group=Marvel-Series, parameterName=Streak-R-V-Offset}
   )
  )
  n10 = TextureObject() {texture=/Game/Global/Textures/Default/Noise/T_Marvel_Series}
  n11 = ScalarParameter() {defaultValue=3, group=Marvel-Series, parameterName=Streak-R-Power-Max}
  n12 = ScalarParameter() {defaultValue=2, group=Marvel-Series, parameterName=Streak-R-Power-Min}
  n13 = ScalarParameter() {defaultValue=0.15, group=Marvel-Series, parameterName=Streak Power Angle}
  n14 = ComponentMask(None=Panner(Coordinate=Rotator(Coordinate=@n6, Time=@n13) {speed=6.2832}, Time=@n8) {speedX=1}) {r=true}
  n15 = Saturate(None=Add(A=Sine(None=@n14), B=ScalarParameter() {defaultValue=0.5, group=Marvel-Series, parameterName=Sine Offset}))
  n16 = Multiply(
   A=Power(
    Base=TextureSample(UVs=@n9, Tex=@n10) {texture=&T_Marvel_Series}.R
    Exp=LinearInterpolate(A=@n11, B=@n12, Alpha=@n15)
   )
   B=LinearInterpolate(
    A=ScalarParameter() {defaultValue=1, group=Marvel-Series, parameterName=Streak-R-Multiply-Max}
    B=ScalarParameter() {defaultValue=0.1, group=Marvel-Series, parameterName=Streak-R-Multiply-Min}
    Alpha=@n15
   )
  )
  n17 = AppendVector(
   A=ScalarParameter() {defaultValue=0.5, group=Marvel-Series, parameterName=Streak-G-U-Tile}
   B=ScalarParameter() {defaultValue=0.125, group=Marvel-Series, parameterName=Streak-G-V-Tile}
  )
  n18 = Add(
   A=Panner(Coordinate=Multiply(A=@n7, B=@n17), Time=@n8) {speedY=0.5}
   B=AppendVector(
    A=ScalarParameter() {defaultValue=0.35, group=Marvel-Series, parameterName=Streak-G-U-Offset}
    B=ScalarParameter() {defaultValue=0.5, group=Marvel-Series, parameterName=Streak-G-V-Offset}
   )
  )
  n19 = ScalarParameter() {defaultValue=3, group=Marvel-Series, parameterName=Streak-G-Power-Max}
  n20 = ScalarParameter() {defaultValue=2, group=Marvel-Series, parameterName=Streak-G-Power_Min}
  n21 = Multiply(
   A=Power(
    Base=TextureSample(UVs=@n18, Tex=@n10) {texture=&T_Marvel_Series}.G
    Exp=LinearInterpolate(A=@n19, B=@n20, Alpha=@n15)
   )
   B=LinearInterpolate(
    A=ScalarParameter() {defaultValue=1, group=Marvel-Series, parameterName=Streak-G-Multiply-Max}
    B=ScalarParameter() {defaultValue=0.1, group=Marvel-Series, parameterName=Streak-G-Multiply-Min}
    Alpha=@n15
   )
  )
  n22 = Saturate(None=Add(A=Saturate(None=@n16), B=Saturate(None=@n21)))
  n23 = Constant2Vector() {g=0.5, r=0.5}
  n24 = ScalarParameter() {defaultValue=0.6, group=Marvel-Series, parameterName=Radius-Enhanced}
  n25 = ScalarParameter() {defaultValue=1.5, group=Marvel-Series, parameterName=Density-Enhanced}
  n26 = MaterialFunctionCall(Float2=TextureCoordinate() {coordinateIndex=3}) {materialFunction=&BreakOutFloat2Components}
  n27 = Divide(A=@n26.R, B=Max(A=@n26.G))
  n28 = Constant() {r=1}
  n29 = OneMinus(None=@n27)
  n30 = Add(
   A=Multiply(A=AppendVector(A=@n27, B=@n28), B=TextureCoordinate() {coordinateIndex=4})
   B=AppendVector(A=Multiply(A=@n29) {constB=0.5}, B=Constant())
  )
  n31 = Add(
   A=Multiply(A=LinearInterpolate(B=@n1.RGB, Alpha=@n22), B=ScalarParameter() {defaultValue=40, group=Marvel-Series, parameterName=Effect-Bump-Multiply})
   B=LinearInterpolate(
    A=VectorParameter() {parameterName=RGB2, defaultValue=<rgba 0.0723 0.0012 0.0018 1>}.RGB
    B=@n1.RGB
    Alpha=MaterialFunctionCall(UVs=@n30, CenterPosition=@n23, Radius=@n24, Density=@n25) {materialFunction=&RadialGradientExponential}.RadialGradientExponential
   )
  )
  n32 = TextureSampleParameter2D(UVs=@n30) {group=Series, parameterName=SeriesTexture, texture=/Game/Global/Textures/Default/Masks_and_Gradients/T_Ui_LavaSeries_Bg}
  n33 = ScalarParameter() {parameterName=Background}
  Multiply(A=Multiply(A=LinearInterpolate(A=@n31, B=@n32.RGB, Alpha=@n33), B=NamedRerouteUsage()), B=VertexColor().A)
  n34 = MaterialFunctionCall(
   R=ScalarParameter() {group=Layout, parameterName=TopLeftCornerRadius, sortPriority=4}
   G=ScalarParameter() {group=Layout, parameterName=BottomLeftCornerRadius, sortPriority=7}
   B=ScalarParameter() {group=Layout, parameterName=TopRightCornerRadius, sortPriority=5}
   A=ScalarParameter() {group=Layout, parameterName=BottomRightCornerRadius, sortPriority=6}
  ) {materialFunction=&AppendMany}
  n35 = ScalarParameter() {
   defaultValue=1
   desc="If enabled reads corner radius values in pixels.\r\nbool"
   group=Layout
   parameterName=UseValueInPixels
   sliderMax=1
   sortPriority=1
  }
  NamedRerouteDeclaration(None=Multiply(
    A=MaterialFunctionCall(Corner Radius=@n34.RGBA, Radius in Pixels=@n35) {materialFunction=&MF_UI_CoreUI_RoundedMask}.Shape Mask
    B=ScalarParameter() {defaultValue=1, group=Color, parameterName=Additive Percentage, sliderMax=1, sortPriority=3}
   )) {name=ShapeMask, nodeColor=<rgba 0 0 0 1>}
  NamedRerouteUsage()
 ]
}
