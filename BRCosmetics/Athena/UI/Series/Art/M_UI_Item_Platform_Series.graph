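# UI material for a Series item platform, drawing an animated circuit-board backdrop with alpha-composite blending.
# An aspect-corrected, skewed, rotated mask texture, offset by per-widget random parameters, supplies tinted dots and gradient lines.
# Its blue channel staggers the timing of a panning gradient highlight, and a sphere mask dims the edges.
# The pattern adds over vignette and vertical-gradient background colors, then multiplies by a rounded-corner mask and vertex alpha.
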Material /BRCosmetics/Athena/UI/Series/Art/M_UI_Item_Platform_Series
graph {
 properties {
  bUsedWithStaticMesh=false
  blendMode=BLEND_AlphaComposite
  materialDomain=MD_UI
  pixelDepthOffsetMode=PDOM_Legacy
  usedShadingModels=Default Lit
 }
 nodes [
  n1 = ScreenPosition()
  Add(A=@n1.ViewportUV, B=ComponentMask(None=ObjectPositionWS()) {g=true, r=true})
  MaterialFunctionCall(Float2=MaterialFunctionCall() {materialFunction=&GetUserInterfaceUV}.Pixel Size) {materialFunction=&BreakOutFloat2Components}
  Multiply(A=@n1.ViewportUV, B=ScalarParameter() {defaultValue=2, parameterName=ScreenUVSize})
  n2 = TextureCoordinate()
  n3 = Constant2Vector() {g=0.5, r=0.5}
  n4 = ScalarParameter() {defaultValue=0.4865, parameterName=SpheremaskRadius}
  n5 = ScalarParameter() {parameterName=SpheremaskHarness}
  n6 = MaterialFunctionCall(Float2=MaterialFunctionCall() {materialFunction=&GetUserInterfaceUV}.Pixel Size) {materialFunction=&BreakOutFloat2Components}
  n7 = Max(A=@n6.G)
  n8 = MaterialFunctionCall(Float2=TextureCoordinate() {coordinateIndex=4}) {materialFunction=&BreakOutFloat2Components}
  n9 = AppendVector(A=Multiply(A=Divide(A=@n6.R, B=@n7), B=@n8.R), B=@n8.G)
  n10 = ScalarParameter() {defaultValue=0.8205, parameterName=SizeCircuitPattern}
  n11 = Multiply(A=ComponentMask(None=@n9) {g=true}, B=ScalarParameter() {defaultValue=0.0022, parameterName=Skew})
  n12 = Multiply(A=@n9, B=AppendVector(A=Add(A=@n10, B=@n11), B=@n10))
  n13 = Subtract(A=ScalarParameter() {parameterName=RandomByWidget}) {constB=0.5}
  n14 = ScalarParameter() {defaultValue=1, parameterName=RandomOffsetX}
  n15 = Subtract(A=ScalarParameter() {parameterName=RandomByWidget2}) {constB=0.5}
  n16 = ScalarParameter() {defaultValue=1, parameterName=RandomOffsetY}
  n17 = Add(
   A=AppendVector(
    A=ScalarParameter() {defaultValue=0.5225, parameterName=CircuitPatternOffset_X}
    B=ScalarParameter() {defaultValue=2.0996, parameterName=CircuitPatternOffset_Y}
   )
   B=AppendVector(A=Multiply(A=@n13, B=@n14), B=Multiply(A=@n15, B=@n16))
  )
  n18 = TextureSample(UVs=Rotator(Coordinate=Add(A=@n12, B=@n17), Time=ScalarParameter() {defaultValue=1, parameterName=RotationCircuitPattern})) {samplerType=SAMPLERTYPE_Masks, texture=/Game/Global/Textures/Default/Masks_and_Gradients/CircuitBoardSeries}
  n19 = MaterialFunctionCall(Float3=Power(Base=@n18.RGB, Exp=ScalarParameter() {defaultValue=1.5881, parameterName=ContrastGlow})) {materialFunction=&BreakOutFloat3Components}
  n20 = Multiply(A=VectorParameter() {parameterName=ColorCircuitDots, defaultValue=<rgba 0.6008 0.2448 2 1>}.RGB, B=@n19.R)
  n21 = ComponentMask(None=TextureCoordinate()) {g=true}
  n22 = ScalarParameter() {defaultValue=1.4, parameterName=ContrastLineGradient}
  n23 = Multiply(
   A=LinearInterpolate(
    A=VectorParameter() {parameterName=ColorCircuitLines, defaultValue=<rgba 0.099 0 0.5302 1>}.RGB
    B=VectorParameter() {parameterName=ColorCircuitLines2, defaultValue=<rgba 1.7038 1.3982 10 1>}.RGB
    Alpha=Power(Base=@n21, Exp=@n22)
   )
   B=@n19.G
  )
  n24 = VectorParameter() {parameterName=ColorCircuitHighlights, defaultValue=<rgba 0 0.2947 2 1>}
  n25 = TextureCoordinate()
  n26 = ScalarParameter() {defaultValue=11.4747, parameterName=Timeooffset}
  n27 = Multiply(A=Add(A=Multiply(A=@n18.B, B=@n26), B=Time()), B=ScalarParameter() {defaultValue=0.2198, parameterName=SpeedPanHighlights})
  n28 = ScalarParameter() {defaultValue=3, parameterName=DotBirghtness}
  n29 = Multiply(
   A=TextureSample(UVs=Add(A=@n25, B=@n27)) {samplerType=SAMPLERTYPE_Masks, texture=/Game/Global/Textures/Default/Masks_and_Gradients/SeriesGradient}.R
   B=Add(A=Multiply(A=@n19.R, B=@n28), B=@n19.G)
  )
  n30 = Multiply(
   A=LinearInterpolate(A=ScalarParameter() {defaultValue=0.0686, parameterName=MinBrightnesssSpheremask}, Alpha=SphereMask(A=@n2, B=@n3, Radius=@n4, Hardness=@n5))
   B=Add(A=Add(A=@n20, B=@n23), B=Multiply(A=@n24.RGB, B=@n29))
  )
  n31 = TextureCoordinate()
  n32 = ScalarParameter() {defaultValue=1, parameterName=Vignette_G_Scale_X}
  n33 = Constant() {r=1}
  n34 = MaterialFunctionCall(
   UVs=MaterialFunctionCall(UVs=@n31, Texture Scale=AppendVector(A=@n32, B=@n33)) {materialFunction=&ScaleUVsByCenter}.UVs
   CenterPosition=AppendVector(A=Constant() {r=0.5}, B=ScalarParameter() {defaultValue=0.5, parameterName=Vignette_G_Position_Y})
   Radius=ScalarParameter() {defaultValue=0.5, parameterName=Vignette_G_Radius}
   Density=ScalarParameter() {defaultValue=2, parameterName=Vignette_G_Power}
  ) {materialFunction=&RadialGradientExponential}
  n35 = Constant() {r=1}
  n36 = ScalarParameter() {defaultValue=1.5, parameterName=Vignette_R_Scale_Y}
  n37 = MaterialFunctionCall(
   UVs=MaterialFunctionCall(UVs=@n31, Texture Scale=AppendVector(A=@n35, B=@n36)) {materialFunction=&ScaleUVsByCenter}.UVs
   CenterPosition=Constant() {r=0.5}
   Radius=ScalarParameter() {defaultValue=0.5, parameterName=Vignette_R_Radius}
   Density=ScalarParameter() {defaultValue=2, parameterName=Vignette_R_Power}
  ) {materialFunction=&RadialGradientExponential}
  n38 = ComponentMask(None=TextureCoordinate()) {g=true}
  n39 = ScalarParameter() {defaultValue=7.0551, parameterName=ContrastBackground2}
  n40 = Add(
   A=Multiply(
    A=LinearInterpolate(B=@n34.RadialGradientExponential, Alpha=@n37.RadialGradientExponential)
    B=VectorParameter() {parameterName=ColorCircuitBackground, defaultValue=<rgba 1 0.1902 0.9706 1>}.RGB
   )
   B=Multiply(A=Power(Base=@n38, Exp=@n39), B=VectorParameter() {parameterName=ColorCircuitBackground2, defaultValue=<rgba 0 0.3712 1 1>}.RGB)
  )
  Multiply(A=Multiply(A=Add(A=@n30, B=@n40), B=NamedRerouteUsage()), B=VertexColor().A)
  n41 = MaterialFunctionCall(
   R=ScalarParameter() {group=Layout, parameterName=TopLeftCornerRadius, sortPriority=4}
   G=ScalarParameter() {group=Layout, parameterName=BottomLeftCornerRadius, sortPriority=7}
   B=ScalarParameter() {group=Layout, parameterName=TopRightCornerRadius, sortPriority=5}
   A=ScalarParameter() {group=Layout, parameterName=BottomRightCornerRadius, sortPriority=6}
  ) {materialFunction=&AppendMany}
  n42 = ScalarParameter() {
   defaultValue=1
   desc="If enabled reads corner radius values in pixels.\r\nbool"
   group=Layout
   parameterName=UseValueInPixels
   sliderMax=1
   sortPriority=1
  }
  NamedRerouteDeclaration(None=Multiply(
    A=MaterialFunctionCall(Corner Radius=@n41.RGBA, Radius in Pixels=@n42) {materialFunction=&MF_UI_CoreUI_RoundedMask}.Shape Mask
    B=ScalarParameter() {defaultValue=1, group=Color, parameterName=Additive Percentage, sliderMax=1, sortPriority=3}
   )) {name=ShapeMask, nodeColor=<rgba 0 0 0 1>}
  NamedRerouteUsage()
  Power(Exp=ScalarParameter() {defaultValue=0.8611, parameterName=ContrastBackground})
  TextureCoordinate()
  TextureSample(UVs=Multiply(A=TextureCoordinate())) {samplerType=SAMPLERTYPE_Masks, texture=&SeriesGradient}
 ]
}
