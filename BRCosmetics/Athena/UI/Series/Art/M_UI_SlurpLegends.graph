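# Translucent UI material that draws an animated Slurp Legends series background, with rising and falling sine waves and drifting bubbles over a two-color radial gradient.
# Three channels of T-SlurpLegends pan as separate layers on aspect-corrected, wobbling UVs distorted by blurred noise.
# A horizontal sine thresholds each layer before it is tinted and composited over the waves.
# A rounded-corner function, scaled by an additive percentage, supplies a named ShapeMask.
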Material /BRCosmetics/Athena/UI/Series/Art/M_UI_SlurpLegends
graph {
 properties {
  bUsedWithStaticMesh=false
  blendMode=BLEND_Translucent
  materialDomain=MD_UI
  pixelDepthOffsetMode=PDOM_Legacy
  usedShadingModels=Default Lit
 }
 nodes [
  n1 = LinearInterpolate(
   A=VectorParameter() {parameterName=RGB1, defaultValue=<rgba 0 0.9473 0.9144 1>}.RGB
   B=VectorParameter() {parameterName=RGB2, defaultValue=<rgba 0 0.1207 0.5573 1>}.RGB
   Alpha=MaterialFunctionCall(
    Radius=ScalarParameter() {defaultValue=0.75, parameterName=BackgroundRadius}
    Density=ScalarParameter() {defaultValue=2.9843, parameterName=BackgroundDensity}
   ) {materialFunction=&RadialGradientExponential}.RadialGradientExponential
  )
  n2 = TextureCoordinate()
  n3 = ComponentMask(None=Panner(Coordinate=@n2) {speedX=0.25}) {r=true}
  n4 = Sine(None=Multiply(A=Time(), B=ScalarParameter() {defaultValue=0.125, parameterName=Wave-Time-Factor}))
  n5 = Multiply(
   A=Sine(None=@n3) {period=4}
   B=LinearInterpolate(B=ScalarParameter() {defaultValue=0.15, parameterName=Wave-02-Amplitude}, Alpha=@n4) {constA=0.05, constB=0.06}
  )
  n6 = ScalarParameter() {defaultValue=-0.35, parameterName=Wave-02-Offset}
  n7 = ComponentMask(None=@n2) {g=true}
  n8 = Multiply(A=Add(A=@n7, B=Add(A=@n5, B=@n6))) {constB=64}
  n9 = Multiply(A=Sine(None=@n3) {period=2}, B=LinearInterpolate(B=ScalarParameter() {defaultValue=0.075, parameterName=Wave-01-Amplitude}, Alpha=@n4))
  n10 = ScalarParameter() {defaultValue=-0.1548, parameterName=Wave-01-Offset}
  n11 = Multiply(A=Add(A=@n7, B=Add(A=@n9, B=@n10))) {constB=64}
  n12 = LinearInterpolate(A=OneMinus(None=Saturate(None=@n8)), Alpha=OneMinus(None=Saturate(None=@n11))) {constB=0}
  n13 = LinearInterpolate(A=@n1, B=Multiply(A=@n1, B=ScalarParameter() {defaultValue=1.125, parameterName=Wave-Multiply}), Alpha=@n12)
  n14 = Multiply(A=@n1, B=ScalarParameter() {defaultValue=1.25, parameterName=B-Multiply})
  n15 = MaterialFunctionCall() {materialFunction=&GetUserInterfaceUV}
  n16 = MaterialFunctionCall(Float2=@n15.Pixel Size) {materialFunction=&BreakOutFloat2Components}
  n17 = Max(A=@n16.G) {constB=0.1}
  n18 = Multiply(A=AppendVector(A=Divide(A=@n16.R, B=@n17), B=Constant() {r=1}), B=@n15.Normalized UV)
  n19 = Panner(Coordinate=@n18) {speedX=0.25}
  n20 = Sine(None=Multiply(A=Time(), B=ScalarParameter() {defaultValue=0.125, parameterName=BubbleWave-Time-Factor}))
  n21 = Multiply(
   A=Sine(None=ComponentMask(None=@n19) {g=true}) {period=2}
   B=LinearInterpolate(B=ScalarParameter() {defaultValue=0.001, parameterName=Bubble-Wave}, Alpha=@n20) {constA=0.05, constB=0.06}
  )
  n22 = ScalarParameter() {defaultValue=0.05, parameterName=Bubble-Wave-Offset}
  n23 = LinearInterpolate(A=Multiply(A=@n22) {constB=-1}, B=@n22, Alpha=@n20)
  n24 = AppendVector(A=Add(A=ComponentMask(None=@n18) {r=true}, B=Add(A=@n21, B=@n23)), B=ComponentMask(None=@n18) {g=true})
  n25 = Multiply(A=@n24, B=ScalarParameter() {defaultValue=1, parameterName=B-Tile})
  n26 = Multiply(A=Time(), B=ScalarParameter() {defaultValue=0.5, parameterName=TimeFactor-Bubbles})
  n27 = AppendVector(A=ScalarParameter() {defaultValue=0.125, parameterName=B-X-Speed}, B=ScalarParameter() {defaultValue=0.125, parameterName=B-Y-Speed})
  n28 = TextureCoordinate()
  n29 = AppendVector(A=ScalarParameter() {defaultValue=0.5, parameterName=U-Tile}, B=ScalarParameter() {defaultValue=3, parameterName=V-Tile})
  n30 = Multiply(
   A=TextureSample(UVs=Multiply(A=@n28, B=@n29)) {samplerType=SAMPLERTYPE_Grayscale, texture=&LowResBlurredNoise}.R
   B=AppendVector(A=ScalarParameter() {defaultValue=0.025, parameterName=U-Tile-Dis}, B=ScalarParameter() {defaultValue=0.0125, parameterName=V-Tile-Dis})
  )
  n31 = TextureSample(UVs=Add(A=Panner(Coordinate=@n25, Time=@n26, Speed=@n27) {speedX=0.125, speedY=-0.25}, B=@n30)) {texture=/Game/Global/Textures/Default/Noise/T-SlurpLegends}
  n32 = Sine(None=ComponentMask(None=TextureCoordinate()) {r=true}) {period=2}
  n33 = LinearInterpolate(B=ScalarParameter() {defaultValue=0.25, parameterName=Max}, Alpha=@n32) {constA=1, constB=0}
  n34 = ScalarParameter() {defaultValue=3, parameterName=B-EdgeOffset}
  n35 = Saturate(None=Multiply(A=Subtract(A=@n31.B, B=@n33), B=Multiply(A=@n34, B=@n32)))
  n36 = LinearInterpolate(A=@n35, B=Multiply(A=@n35) {constB=0.5}, Alpha=@n12) {constB=0}
  n37 = ScalarParameter() {defaultValue=1.3, parameterName=G-Multiply}
  n38 = Panner(
   Coordinate=Multiply(A=@n24, B=ScalarParameter() {defaultValue=1.25, parameterName=G-Tile})
   Time=@n26
   Speed=AppendVector(A=ScalarParameter() {defaultValue=0.05, parameterName=G-X-Speed}, B=ScalarParameter() {defaultValue=0.25, parameterName=G-Y-Speed})
  ) {speedX=0.125, speedY=-0.25}
  n39 = Subtract(A=TextureSample(UVs=Add(A=@n38, B=@n30)) {texture=&T-SlurpLegends}.G, B=@n33)
  n40 = VectorParameter() {parameterName=RGB1, defaultValue=<rgba 0 0.9473 0.9144 1>}
  n41 = LinearInterpolate(A=@n1, B=@n40.RGB, Alpha=Multiply(A=ScalarParameter() {defaultValue=0.125, parameterName=G-AO-Multiply}, B=@n39))
  n42 = Multiply(A=@n39, B=Multiply(A=ScalarParameter() {defaultValue=12, parameterName=G-EdgeOffset}, B=@n32))
  n43 = Panner(
   Coordinate=Multiply(A=@n24, B=ScalarParameter() {defaultValue=1.5, parameterName=R-Tile})
   Time=@n26
   Speed=AppendVector(A=ScalarParameter() {defaultValue=-0.125, parameterName=R-X-Speed}, B=ScalarParameter() {defaultValue=0.5, parameterName=R-Y-Speed})
  ) {speedX=0.125, speedY=-0.25}
  n44 = Subtract(A=TextureSample(UVs=Add(A=@n43, B=@n30)) {texture=&T-SlurpLegends}.R, B=@n33)
  n45 = Multiply(A=@n44, B=ScalarParameter() {defaultValue=0.175, parameterName=R-AO-Multiply})
  n46 = Multiply(A=ScalarParameter() {defaultValue=12, parameterName=R-EdgeOffset}, B=@n32)
  LinearInterpolate(
   A=LinearInterpolate(A=LinearInterpolate(A=@n13, B=@n14, Alpha=@n36), B=Multiply(A=@n37, B=@n41), Alpha=Saturate(None=@n42))
   B=Multiply(A=LinearInterpolate(A=@n1, B=@n40.RGB, Alpha=@n45), B=ScalarParameter() {defaultValue=1.5041, parameterName=R-Multiply})
   Alpha=Saturate(None=Multiply(A=@n44, B=@n46))
  )
  LinearInterpolate()
  n47 = MaterialFunctionCall(
   R=ScalarParameter() {group=Layout, parameterName=TopLeftCornerRadius, sortPriority=4}
   G=ScalarParameter() {group=Layout, parameterName=BottomLeftCornerRadius, sortPriority=7}
   B=ScalarParameter() {group=Layout, parameterName=TopRightCornerRadius, sortPriority=5}
   A=ScalarParameter() {group=Layout, parameterName=BottomRightCornerRadius, sortPriority=6}
  ) {materialFunction=&AppendMany}
  n48 = ScalarParameter() {
   defaultValue=1
   desc="If enabled reads corner radius values in pixels.\r\nbool"
   group=Layout
   parameterName=UseValueInPixels
   sliderMax=1
   sortPriority=1
  }
  NamedRerouteDeclaration(None=Multiply(
    A=MaterialFunctionCall(Corner Radius=@n47.RGBA, Radius in Pixels=@n48) {materialFunction=&MF_UI_CoreUI_RoundedMask}.Shape Mask
    B=ScalarParameter() {defaultValue=1, group=Color, parameterName=Additive Percentage, sliderMax=1, sortPriority=3}
   )) {name=ShapeMask, nodeColor=<rgba 0 0 0 1>}
 ]
}
